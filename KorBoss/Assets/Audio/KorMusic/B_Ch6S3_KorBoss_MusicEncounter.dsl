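; Music encounter Blueprint for the Kor boss fight (Ch6S3), driving combat music phases from boss health.
; On health changes below 0.4 (the inline comments say 50%), it switches to combat phase 1 if not already there.
; It also feeds health percent to the combat music audio component's "Boss.Hp" parameter for a sweetener layer.
; Entering the completed state sets combat phase 2, and both overridden events call their parent first.
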(event EventEnterCompletedState
  (|CallParentFunction)
  (CallFunction|SetBossEncounterCombatPhase 2))

(event MusicEncounter|Boss|EventBossHealthChanged
  (params (Boss%20Component%20Interface :type Music%20Encounter%20Component%20Boss%20Interface%20Interface) (Health%20Info :type Struct%20NPC%20Health%20Info%20Structure))
  (|CallParentFunction Boss%20Component%20Interface Health%20Info)
  (bind _output_get (Variables|Default|SetHealth%20Percent (Utilities|Struct|BreakStructNPCHealthInfo Health%20Info)))
  (Math|Float|CompareFloat _output_get 0.4
    (:<
      ;[0.6 0.4 0.8] Enter final Phase at 50% hp
      (Utilities|FlowControl|Sequence
        (:then_0
          (bind _combat_phase (MusicEncounterDashboard|Boss|GetBossEncounterCombatPhase))
          (if (!= _combat_phase 1)
            (CallFunction|SetBossEncounterCombatPhase 1)))
        (:then_1
          ;[0.6 0.4 0.8] Update Boss sweetner layer when below 50% hp
          (bind _combat_music_audio_component (Variables|MusicEncounter|Boss|Combat|GetCombat%20Music%20Audio%20Component)
            (:then
              (Audio|Parameter|SetFloatParameter _combat_music_audio_component "Boss.Hp" (Variables|Default|GetHealth%20Percent) :!class /Script/AudioExtensions.AudioParameterControllerInterface))
            (:else)))))))
